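# Physics asset for the BlazerVeil backpack cosmetic, giving the pack, its straps and sword capsule bodies for physics-driven motion.
# A six-link right strap chains to the main pack through a small MSTR body, whose pack joint swings and twists freely while the links have limited ranges.
# The sword and mid strap attach directly to the pack, the sword almost rigidly and the strap with one wider swing axis.
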PhysicsAsset /BRCosmetics/Accessories/FORT_Backpacks/Backpack_F_MED_BlazerVeil/Meshes/F_MED_BlazerVeil_Pack_Physics
bodies {
 dyn_mainPack [{
   shapeType=Capsule
   length=7.0247
   radius=11.4211
   shapeName=dyn_mainPack_capsule
   center=<1.4326 -7.1553 -2.1011>
   rotation=<rot 55.7124 -180 -180>
  }]
 dyn_midStrap_01 [{shapeType=Capsule, length=10.6284, radius=3, shapeName=dyn_midStrap_01_capsule, center=<6.6209 0 5>, rotation=<rot 0 0 90>}]
 dyn_r_Strap_01 [{shapeType=Capsule, length=1.4283, radius=2.1353, shapeName=dyn_r_Strap_01_capsule, center=<2.0328 0 0>, rotation=<rot 90 0 0>}]
 dyn_r_Strap_01_MSTR [{shapeType=Capsule, length=0.505, radius=0.505, shapeName=dyn_r_Strap_01_MSTR_capsule}]
 dyn_r_Strap_02 [{shapeType=Capsule, length=2.1289, radius=1.8711, shapeName=dyn_r_Strap_02_capsule, center=<2.8654 0 0>, rotation=<rot 90 0 0>}]
 dyn_r_Strap_03 [{shapeType=Capsule, length=1.6481, radius=1.7574, shapeName=dyn_r_Strap_03_capsule, center=<2.8654 0 0>, rotation=<rot 90 0 0>}]
 dyn_r_Strap_04 [{shapeType=Capsule, length=0.2142, radius=3.7858, shapeName=dyn_r_Strap_04_capsule, center=<2.8654 0 0>, rotation=<rot 90 0 0>}]
 dyn_r_Strap_05 [{shapeType=Capsule, length=2.2484, radius=1.7516, shapeName=dyn_r_Strap_05_capsule, center=<2.8654 0 0>, rotation=<rot 90 0 0>}]
 dyn_r_Strap_06 [{shapeType=Capsule, length=6.8383, radius=3.4758, shapeName=dyn_r_Strap_06_capsule, center=<2.7356 -2 0>, rotation=<rot 90 -65.0183 -48.2619>}]
 dyn_sword [{shapeType=Capsule, length=32.1578, radius=1.4761, shapeName=dyn_sword_capsule, center=<-0.5273 0 0>, rotation=<rot 90 0 0>}]
}
constraints [
 {bone1Name=dyn_r_Strap_02, bone2Name=dyn_r_Strap_01, swing1LimitDegrees=20, swing2LimitDegrees=30, twistLimitDegrees=20}
 {bone1Name=dyn_r_Strap_03, bone2Name=dyn_r_Strap_02, swing1LimitDegrees=10, swing2LimitDegrees=30, twistLimitDegrees=5}
 {bone1Name=dyn_r_Strap_04, bone2Name=dyn_r_Strap_03, swing1LimitDegrees=10, swing2LimitDegrees=30, twistLimitDegrees=5}
 {bone1Name=dyn_r_Strap_05, bone2Name=dyn_r_Strap_04, swing1LimitDegrees=10, swing2LimitDegrees=30, twistLimitDegrees=5}
 {bone1Name=dyn_r_Strap_06, bone2Name=dyn_r_Strap_05, swing1LimitDegrees=10, swing2LimitDegrees=30, twistLimitDegrees=5}
 {bone1Name=dyn_r_Strap_01, bone2Name=dyn_r_Strap_01_MSTR, swing1LimitDegrees=10, swing2LimitDegrees=30, twistLimitDegrees=5}
 {
  bone1Name=dyn_r_Strap_01_MSTR
  bone2Name=dyn_mainPack
  swing1LimitDegrees=45
  swing1Motion=Free
  swing2LimitDegrees=45
  swing2Motion=Free
  twistLimitDegrees=45
  twistMotion=Free
 }
 {bone1Name=dyn_sword, bone2Name=dyn_mainPack, swing1LimitDegrees=2, swing2LimitDegrees=2, twistLimitDegrees=2}
 {bone1Name=dyn_midStrap_01, bone2Name=dyn_mainPack, swing1LimitDegrees=30, swing2LimitDegrees=5, twistLimitDegrees=5}
]
