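# Animation Blueprint for the BlazerVeil backpack, animating its right strap and toggling a material element.
# bRigidBodyEnabled selects RigidBody physics with state-driven gravity and weighted rotation constraints, or a trail-controller fallback with staggered additive bone bends, both from the cached input pose.
# Weapon, customization and skydive events set "Hide Element 01" when a BlazerVeil-tagged pickaxe is equipped outside skydiving.
# StrapAlpha and NewJumpingFalling update but feed no visible pose node.
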AnimBlueprint /BRCosmetics/Accessories/FORT_Backpacks/Backpack_F_MED_BlazerVeil/Meshes/F_MED_BlazerVeil_Pack_AnimBP
target_skeleton=&/F_MED_BlazerVeil_Pack_Skeleton
graphs [
 AnimGraph {nodes [
   n1 = RigidBody!(
    ComponentPose=Local To Component(LocalPose=Use cached pose 'PreRigid'())
    Alpha
    OverrideWorldGravity:Vector=Get GravityOverride()
    ExternalForce
   ) -> (Pose) {
    bClampLinearTranslationLimitToRefPose=true
    bOverrideWorldGravity=true
    overridePhysicsAsset=&/F_MED_BlazerVeil_Pack_Physics
    simulationSpace=BaseBoneSpace
    alphaBoolBlend {blendOption=Linear}
    baseBoneRef {boneName=pack}
    componentAppliedLinearAccClamp=<800 3000 900>
    componentLinearAccScale=<0.5 0.5 0.45>
    overrideWorldGravity {z=-980}
   }
   n2 = "Constraint - dyn_r_Strap_04 (R)"!(ComponentPose=@n1, bAlphaBoolEnabled:bool=True, ConstraintWeights_0=0.3) -> (Pose) {
    alphaInputType=Bool
    alphaBoolBlend {blendInTime=0.1, blendOption=Linear, blendOutTime=0.1}
    boneToModify {boneName=dyn_r_Strap_04}
    constraintSetup [{transformType=Rotation, targetBone {boneName=dyn_mainPack}}]
    constraintWeights [1]
   }
   n3 = "Constraint - dyn_r_Strap_05 (R)"!(ComponentPose=@n2, bAlphaBoolEnabled:bool=True, ConstraintWeights_0=0.4) -> (Pose) {
    alphaInputType=Bool
    alphaBoolBlend {blendInTime=0.1, blendOption=Linear, blendOutTime=0.1}
    boneToModify {boneName=dyn_r_Strap_05}
    constraintSetup [{transformType=Rotation, targetBone {boneName=dyn_mainPack}}]
    constraintWeights [1]
   }
   n4 = "Constraint - dyn_r_Strap_06 (R)"!(ComponentPose=@n3, bAlphaBoolEnabled:bool=True, ConstraintWeights_0=0.7) -> (Pose) {
    alphaInputType=Bool
    alphaBoolBlend {blendInTime=0.1, blendOption=Linear, blendOutTime=0.1}
    boneToModify {boneName=dyn_r_Strap_06}
    constraintSetup [{transformType=Rotation, targetBone {boneName=dyn_mainPack}}]
    constraintWeights [1]
   }
   n5 = "Transform (Modify) Bone - Bone: dyn_r_Strap_01_MSTR"(ComponentPose=Local To Component(LocalPose=Use cached pose 'PreRigid'()), Alpha=1, Translation="0,0,0", Rotation="-20,0,0", Scale="1,1,1") {
    rotationMode=BMM_Additive
    rotationSpace=BCS_BoneSpace
    scaleMode=BMM_Additive
    scaleSpace=BCS_BoneSpace
    translationMode=BMM_Additive
    translationSpace=BCS_BoneSpace
    alphaBoolBlend {blendOption=Linear}
    boneToModify {boneName=dyn_r_Strap_01_MSTR}
    rotation {pitch=-20}
   }
   n6 = "Transform (Modify) Bone - Bone: dyn_r_Strap_04"(ComponentPose=@n5, Alpha=1, Translation="0,0,0", Rotation="15,0,0", Scale="1,1,1") {
    rotationMode=BMM_Additive
    rotationSpace=BCS_BoneSpace
    scaleMode=BMM_Additive
    scaleSpace=BCS_BoneSpace
    translationMode=BMM_Additive
    translationSpace=BCS_BoneSpace
    alphaBoolBlend {blendOption=Linear}
    boneToModify {boneName=dyn_r_Strap_04}
    rotation {pitch=15}
   }
   n7 = Trail controller - Bone: dyn_r_Strap_06!(ComponentPose=@n6, Alpha=0.7, FakeVelocity:Vector="0,0,0") -> (Pose) {
    bLimitRotation=true
    bShowTrailLocation=true
    chainLength=7
    debugLifeTime=0.1
    maxDeltaTime=0.0333
    alphaScaleBiasClamp {bInterpResult=true}
    baseJoint {boneName=root}
    planarLimits [{drivingBone {boneName=pack}, planeTransform {location=<0 -8 0>, rotation=<rot 0 0 -11.0409>, scale=<1 1 1>}}]
    rotationLimits [
     {limitMax=<45 45 2>, limitMin=<0 -45 -45>}
     {limitMax=<45 45 2>, limitMin=<0 -45 -45>}
     {limitMax=<45 45 2>, limitMin=<0 -45 -45>}
     {limitMax=<45 45 2>, limitMin=<0 -45 -45>}
     {limitMax=<45 45 2>, limitMin=<0 -45 -45>}
     {limitMax=<45 45 2>, limitMin=<0 -45 -45>}
     {limitMax=<45 45 2>, limitMin=<0 -45 -45>}
    ]
    rotationOffsets [<0 0 0>, <0 0 0>, <0 0 0>, <0 0 0>, <0 0 0>, <0 0 0>, <0 0 0>]
    trailBone {boneName=dyn_r_Strap_06}
    trailRelaxationSpeed {editorCurveData {keys [{arriveTangent=-1.7403, interpMode=RCIM_Cubic, time=0.0206, value=30}, <at 0.9722 15 RCIM_Cubic>]}}
   }
   n8 = "Transform (Modify) Bone - Bone: dyn_r_Strap_01_MSTR"!(ComponentPose=@n7, bAlphaBoolEnabled:bool=True, Translation="0,0,0", Rotation="1,0,0", Scale="1,1,1") -> (Pose) {
    alphaInputType=Bool
    rotationMode=BMM_Additive
    rotationSpace=BCS_BoneSpace
    scaleMode=BMM_Additive
    scaleSpace=BCS_BoneSpace
    translationMode=BMM_Additive
    translationSpace=BCS_BoneSpace
    alphaBoolBlend {blendInTime=0.5, blendOption=Linear}
    boneToModify {boneName=dyn_r_Strap_01_MSTR}
    rotation {pitch=1}
   }
   n9 = "Transform (Modify) Bone - Bone: dyn_r_Strap_02"!(ComponentPose=@n8, bAlphaBoolEnabled:bool=True, Translation="0,0,0", Rotation="10,0,0", Scale="1,1,1") -> (Pose) {
    alphaInputType=Bool
    rotationMode=BMM_Additive
    rotationSpace=BCS_BoneSpace
    scaleMode=BMM_Additive
    scaleSpace=BCS_BoneSpace
    translationMode=BMM_Additive
    translationSpace=BCS_BoneSpace
    alphaBoolBlend {blendInTime=1, blendOption=Linear}
    boneToModify {boneName=dyn_r_Strap_02}
    rotation {pitch=10}
   }
   n10 = "Transform (Modify) Bone - Bone: dyn_r_Strap_03"!(ComponentPose=@n9, bAlphaBoolEnabled:bool=True, Translation="0,0,0", Rotation="25,0,0", Scale="1,1,1") -> (Pose) {
    alphaInputType=Bool
    rotationMode=BMM_Additive
    rotationSpace=BCS_BoneSpace
    scaleMode=BMM_Additive
    scaleSpace=BCS_BoneSpace
    translationMode=BMM_Additive
    translationSpace=BCS_BoneSpace
    alphaBoolBlend {blendInTime=1.2, blendOption=Linear}
    boneToModify {boneName=dyn_r_Strap_03}
    rotation {pitch=25}
   }
   n11 = "Transform (Modify) Bone - Bone: dyn_r_Strap_04"!(ComponentPose=@n10, bAlphaBoolEnabled:bool=True, Translation="0,0,0", Rotation="30,0,0", Scale="1,1,1") -> (Pose) {
    alphaInputType=Bool
    rotationMode=BMM_Additive
    rotationSpace=BCS_BoneSpace
    scaleMode=BMM_Additive
    scaleSpace=BCS_BoneSpace
    translationMode=BMM_Additive
    translationSpace=BCS_BoneSpace
    alphaBoolBlend {blendInTime=1.4, blendOption=Linear}
    boneToModify {boneName=dyn_r_Strap_04}
    rotation {pitch=30}
   }
   n12 = "Transform (Modify) Bone - Bone: dyn_r_Strap_05"!(ComponentPose=@n11, bAlphaBoolEnabled:bool=True, Translation="0,0,0", Rotation="35,0,0", Scale="1,1,1") -> (Pose) {
    alphaInputType=Bool
    rotationMode=BMM_Additive
    rotationSpace=BCS_BoneSpace
    scaleMode=BMM_Additive
    scaleSpace=BCS_BoneSpace
    translationMode=BMM_Additive
    translationSpace=BCS_BoneSpace
    alphaBoolBlend {blendInTime=1.6, blendOption=Linear}
    boneToModify {boneName=dyn_r_Strap_05}
    rotation {pitch=35}
   }
   n13 = "Transform (Modify) Bone - Bone: dyn_r_Strap_06"!(ComponentPose=@n12, bAlphaBoolEnabled:bool=True, Translation="0,0,0", Rotation="40,0,0", Scale="1,1,1") -> (Pose) {
    alphaInputType=Bool
    rotationMode=BMM_Additive
    rotationSpace=BCS_BoneSpace
    scaleMode=BMM_Additive
    scaleSpace=BCS_BoneSpace
    translationMode=BMM_Additive
    translationSpace=BCS_BoneSpace
    alphaBoolBlend {blendInTime=1.8, blendOption=Linear}
    boneToModify {boneName=dyn_r_Strap_06}
    rotation {pitch=40}
   }
   Output Pose(Result=Blend Poses by bool(BlendPose_0=Component To Local(ComponentPose=@n4), BlendPose_1=Component To Local(ComponentPose=@n13), bActiveValue=Get bRigidBodyEnabled()) {blendType=Linear}) {layerGroup=None}
   Save cached pose 'PreRigid'(Pose=Input Pose()) {cacheName=PreRigid}
  ]}
 EventGraph {nodes [
   Event BlueprintUpdateAnimation()
   > n1 = Sequence!(execute) -> (then_0, then_1, then_2:Exec, then_3:Exec)
   > n2 = Gravity Override State Function DEPRECATED(CharacterPartAnimInstance=Self-Reference(), GameTimeInSeconds=GetGameTimeInSeconds(), GravityOverrideParams=Get Gravity Override Params())
   > Set GravityOverride(GravityOverride=@n2)
   n3 = Branch(execute=@n1.then_1, Condition=OR Boolean(A=Get bIsJumping(), B=Get bIsFalling()))
   > Set NewJumpingFalling(NewJumpingFalling:bool=true) -> (Output_Get:bool)
   Set NewJumpingFalling(execute=@n3.else, NewJumpingFalling=false)
   n4 = Branch(execute=@n1.then_2, Condition=OR Boolean(A=Get bIsJumping(), B=Get bIsFalling()))
   > Set StrapAlpha(StrapAlpha:double=0) -> (Output_Get:double)
   Set StrapAlpha(execute=@n4.else, StrapAlpha=0.7)
   Branch(execute=@n1.then_3, Condition=Get bIsInFrontEnd())
   > Set StrapAlpha(StrapAlpha=0.3)
   n5 = Event OnWeaponChanged()
   n6 = Get Dynamic Material Instance(execute=[@n5.then, @n12.then], SkeletalMeshComponent=GetPartSkeletalMeshComponent())
   n7 = PickaxeItemDefinitionHasAnyMetaTag(Tags=Make Array("[0]"=(TagName="Cosmetics.MetaTags.BlazerVeil")), bRequiresEquipped=true)
   > n8 = Branch(Condition=@n7)
   > n9 = Branch(Condition=Get bIsSkyDiving())
   SetScalarParameterValue(execute=[@n10.then, @n9.then, @n11.else, @n8.else], self=@n6, ParameterName=Hide Element 01, Value=0)
   SetScalarParameterValue(execute=[@n9.else, @n11.then], self=@n6, ParameterName=Hide Element 01, Value=1)
   n10 = Event OnBeginSkydiving()
   Event OnEndSkydiving()
   > n11 = Branch(Condition=@n7)
   n12 = Event OnCharacterCustomizationLoaded()
   TryGetPawnOwner()
  ]}
]
variables [
 Gravity Override Params {type=GravityOverrideParamsStruct, default {skydive_GoingUp {gravityOverride=<0 -300 -150>}}}
 GravityOverride {type=Vector}
 NewJumpingFalling {type=boolean}
 StrapAlpha {type=number}
]
